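# Debug material in the melee weapon trail FX set, flagged for use on Niagara ribbons, apparently for visualizing UVs as color.
# It appends a constant to the texture coordinates and passes the resulting vector to the Emissive input of the MF_FX_EmissiveAdjustments function.
# The shading model is Default Lit, and the graph shows no textures, parameters or link from the function to a material output.
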Material /CRD_Melee/FX/WeaponTrails/M_DEBUG
graph {
 properties {bUsedWithNiagaraRibbons=true, pixelDepthOffsetMode=PDOM_Legacy, usedShadingModels=Default Lit}
 nodes [MaterialFunctionCall(Emissive=AppendVector(A=TextureCoordinate(), B=Constant())) {materialFunction=&MF_FX_EmissiveAdjustments}]
}
