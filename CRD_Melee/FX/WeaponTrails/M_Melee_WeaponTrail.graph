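# Translucent, unlit, two-sided material for melee weapon trails on Niagara ribbons and mesh particles.
# Particle color feeds emissive through an adjustments function, while particle alpha scales the summed opacity layers.
# A hard trail texture, panned by a dynamic parameter, combines with a curved soft glow carrying scrolling small particles and two-layer motion lines.
# UV gradients and sphere masks shape the edges, and a static debug switch shows opaque UV stripes.
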Material /CRD_Melee/FX/WeaponTrails/M_Melee_WeaponTrail
graph {
 properties {
  bUsedWithNiagaraMeshParticles=true
  bUsedWithNiagaraRibbons=true
  blendMode=BLEND_Translucent
  pixelDepthOffsetMode=PDOM_Legacy
  shadingModel=MSM_Unlit
  twoSided=true
  usedShadingModels=Unlit
 }
 nodes [
  n1 = Sine(None=ComponentMask(None=TextureCoordinate()) {r=true}) {period=0.05}
  n2 = Constant()
  n3 = ParticleColor()
  n4 = StaticBoolParameter() {parameterName=DEBUG MODE}
  MaterialFunctionCall(Emissive=StaticSwitch(True=Step(Y=@n1, X=@n2), False=@n3.RGB, Value=@n4)) {materialFunction=&MF_FX_EmissiveAdjustments}
  n5 = DynamicParameter() {defaultValue=<rgba 0 0 0 0>, paramNames [Pan,Param2,Param3,Param4]}
  n6 = Constant()
  n7 = Saturate(None=Add(A=TextureCoordinate(), B=AppendVector(A=@n5.Pan, B=@n6)))
  n8 = ComponentMask(None=TextureCoordinate()) {r=true}
  n9 = ComponentMask(None=TextureCoordinate()) {g=true}
  n10 = Multiply(
   A=Power(Base=OneMinus(None=@n8), Exp=ScalarParameter() {defaultValue=3, group=Texture - Edge Gradient, parameterName=Hard Trailing Edge Power})
   B=Power(Base=OneMinus(None=@n9), Exp=ScalarParameter() {defaultValue=2, group=Texture - Edge Gradient, parameterName=Hard Handle Edge Power})
  )
  n11 = TextureCoordinate()
  n12 = ScalarParameter() {defaultValue=0.1, group=Texture - Edge Gradient, parameterName=Hard Leading Edge Fade, sliderMax=1}
  n13 = Saturate(None=Multiply(A=ComponentMask(None=@n11) {r=true}, B=Divide(B=@n12)))
  n14 = Sine(None=Add(A=Multiply(A=@n13) {constB=0.5}) {constB=0.25})
  n15 = TextureCoordinate()
  n16 = ScalarParameter() {defaultValue=0.1, group=Texture - Edge Gradient, parameterName=Hard Tip Edge Fade, sliderMax=1}
  n17 = Saturate(None=Multiply(A=ComponentMask(None=@n15) {g=true}, B=Divide(B=@n16)))
  n18 = Sine(None=Add(A=Multiply(A=@n17) {constB=0.5}) {constB=0.25})
  n19 = Multiply(A=OneMinus(None=ConstantBiasScale(None=@n14)), B=OneMinus(None=ConstantBiasScale(None=@n18)))
  n20 = Multiply(
   A=Multiply(
    A=TextureSampleParameter2D(UVs=@n7) {group=Texture, parameterName=Texture, sortPriority=2, texture=&/T_Melee_AnimTrail}.G
    B=Multiply(A=@n10, B=@n19)
   )
   B=ScalarParameter() {defaultValue=0.4, group=Texture, parameterName=Texture Strength}
  )
  n21 = Multiply(A=Time() {bOverride_Period=true, period=3600}, B=ScalarParameter() {defaultValue=2, parameterName=Small Particles Speed})
  n22 = Constant()
  n23 = TextureSampleParameter2D(UVs=Add(A=TextureCoordinate(), B=AppendVector(A=@n21, B=@n22))) {group=Texture, parameterName=Texture, sortPriority=2, texture=&T_Melee_AnimTrail}
  n24 = ComponentMask(None=TextureCoordinate()) {r=true}
  n25 = ComponentMask(None=TextureCoordinate()) {g=true}
  n26 = Multiply(
   A=Power(Base=OneMinus(None=@n24), Exp=ScalarParameter() {defaultValue=2, group=Soft Glow - Edge Gradient, parameterName=Soft Trailing Edge Power})
   B=Power(Base=OneMinus(None=@n25), Exp=ScalarParameter() {defaultValue=2, group=Soft Glow - Edge Gradient, parameterName=Soft Handle Edge Power})
  )
  n27 = TextureCoordinate()
  n28 = ScalarParameter() {defaultValue=0.4, group=Soft Glow - Edge Gradient, parameterName=Soft Leading Edge Fade, sliderMax=1}
  n29 = Saturate(None=Multiply(A=ComponentMask(None=@n27) {r=true}, B=Divide(B=@n28)))
  n30 = Sine(None=Add(A=Multiply(A=@n29) {constB=0.5}) {constB=0.25})
  n31 = TextureCoordinate()
  n32 = ScalarParameter() {defaultValue=0.4, group=Soft Glow - Edge Gradient, parameterName=Soft Tip Edge Fade, sliderMax=1}
  n33 = Saturate(None=Multiply(A=ComponentMask(None=@n31) {g=true}, B=Divide(B=@n32)))
  n34 = Sine(None=Add(A=Multiply(A=@n33) {constB=0.5}) {constB=0.25})
  n35 = Multiply(A=OneMinus(None=ConstantBiasScale(None=@n30)), B=OneMinus(None=ConstantBiasScale(None=@n34)))
  n36 = SphereMask(
   A=TextureCoordinate()
   B=ScalarParameter() {defaultValue=1, group=Soft Glow - Curvature, parameterName=Curvature Pos}
   Radius=ScalarParameter() {defaultValue=1, group=Soft Glow - Curvature, parameterName=Curvature Radius}
   Hardness=ScalarParameter() {defaultValue=0.5, group=Soft Glow - Curvature, parameterName=Curvature hardness}
  )
  n37 = TextureCoordinate()
  n38 = Clamp(None=ComponentMask(None=@n37) {r=true}) {maxDefault=0.5}
  n39 = Power(
   Base=ComponentMask(None=@n37) {g=true}
   Exp=ScalarParameter() {defaultValue=0.35, group=Soft Glow - Roundness, parameterName=Roundness Distort Exp}
  )
  n40 = Multiply(
   A=Multiply(A=Multiply(A=@n26, B=@n35), B=OneMinus(None=@n36))
   B=SphereMask(
    A=AppendVector(A=@n38, B=@n39)
    B=Constant() {r=0.5}
    Radius=ScalarParameter() {defaultValue=0.5, group=Soft Glow - Roundness, parameterName=Roundness Radius}
    Hardness=ScalarParameter() {group=Soft Glow - Roundness, parameterName=Roundness Hardness}
   )
  )
  n41 = VectorParameter() {group=Motion Lines, parameterName=Fine Motion Lines UV Scale A, sortPriority=5, defaultValue=<rgba 1 3.464 0 0>}
  n42 = TextureCoordinate()
  n43 = VectorParameter() {group=Motion Lines, parameterName=Fine Motion Lines Panner Speed A, sortPriority=4, defaultValue=<rgba 0 0.976 0 0>}
  n44 = Panner(Coordinate=Multiply(A=@n42, B=ComponentMask(None=@n41.RGB) {g=true, r=true}), Speed=ComponentMask(None=@n43.RGB) {g=true, r=true})
  n45 = MaterialFunctionCall(XY=@n44) {materialFunction=&Swizzle}
  n46 = ScalarParameter() {defaultValue=1, group=Motion Lines, parameterName=Motion Lines Swizzle UVs, sortPriority=12}
  n47 = NamedRerouteUsage()
  n48 = TextureSampleParameter2D(UVs=Add(A=LinearInterpolate(A=@n44, B=@n45.YX, Alpha=@n46), B=@n47)) {
   group=Motion Lines
   parameterName=Motion Lines Texture
   samplerSource=SSM_Wrap_WorldGroupSettings
   texture=/Game/Effects/Fort_Effects/Textures/Outlander/T_InTheZone_LightRays_01
  }
  n49 = Multiply(
   A=ScalarParameter() {defaultValue=1, group=Motion Lines, parameterName=Motion Lines A Intensity, sortPriority=3}
   B=Power(
    Base=ChannelMaskParameter(None=@n48.RGBA) {group=Motion Lines, parameterName=Motion Lines Texture Mask A, sortPriority=1}.R
    Exp=ScalarParameter() {defaultValue=1.5, group=Motion Lines, parameterName=Motion Lines Texture A Contrast, sortPriority=2}
   ) {constExponent=1.5}
  )
  n50 = VectorParameter() {group=Motion Lines, parameterName=Fine Motion Lines UV Scale B, sortPriority=11, defaultValue=<rgba 1 4.4362 0 0>}
  n51 = VectorParameter() {group=Motion Lines, parameterName=Fine Motion Lines Panner Speed B, sortPriority=10, defaultValue=<rgba 0 -0.7616 0 0>}
  n52 = Panner(Coordinate=Multiply(A=@n42, B=ComponentMask(None=@n50.RGB) {g=true, r=true}), Speed=ComponentMask(None=@n51.RGB) {g=true, r=true}) {speedY=-0.7616}
  n53 = MaterialFunctionCall(XY=@n52) {materialFunction=&Swizzle}
  n54 = TextureSampleParameter2D(UVs=Add(A=LinearInterpolate(A=@n52, B=@n53.YX, Alpha=@n46), B=@n47)) {
   group=Motion Lines
   parameterName=Motion Lines Texture
   samplerSource=SSM_Wrap_WorldGroupSettings
   sortPriority=6
   texture=&T_InTheZone_LightRays_01
  }
  n55 = ChannelMaskParameter(None=@n54.RGBA) {group=Motion Lines, maskChannel=Green, parameterName=Motion Lines Texture Mask B, sortPriority=7, defaultValue=<rgba 0 1 0 0>}
  n56 = Multiply(
   A=Power(Base=@n55.R, Exp=ScalarParameter() {defaultValue=2, group=Motion Lines, parameterName=Motion Lines Texture B Contrast, sortPriority=8})
   B=ScalarParameter() {defaultValue=1, group=Motion Lines, parameterName=Motion Lines B Intensity, sortPriority=9}
  )
  n57 = LinearInterpolate(
   A=ScalarParameter() {defaultValue=0.5, group=Alpha, parameterName=Fine Motion Lines Opacity A, sortPriority=10}
   B=ScalarParameter() {defaultValue=1, group=Alpha, parameterName=Fine Motion Lines Opacity B, sortPriority=11}
   Alpha=Multiply(A=Add(A=@n49, B=@n56), B=ScalarParameter() {defaultValue=0.5, group=Motion Lines, parameterName=Combined Motion Lines Scale})
  ) {constA=1}
  n58 = Add(
   A=Multiply(A=Multiply(A=@n23.R, B=@n40), B=ScalarParameter() {defaultValue=2, parameterName=Small Particles})
   B=Multiply(A=Multiply(A=@n40, B=@n57), B=ScalarParameter() {defaultValue=0.55, group=Soft Glow, parameterName=Soft Component})
  )
  StaticSwitch(True=Constant() {r=1}, False=Multiply(A=@n3.A, B=Add(A=@n20, B=@n58)) {constA=1}, Value=@n4)
 ]
}
