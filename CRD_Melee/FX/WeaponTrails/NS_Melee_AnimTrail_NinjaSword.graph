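# Niagara system that draws the ninja sword's melee weapon trail as short-lived ribbons along an animated skeletal mesh.
# A single AnimTrail emitter uses shared anim-trail modules to spawn, initialize from bone or socket indices, and update particles for three ribbon renderers.
# TrailColor combines time-of-day tints with an optional user custom color.
# Age-based curves darken and fade the trail, scaled by user opacity; dynamic material parameters are disabled.
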NiagaraSystem /CRD_Melee/FX/WeaponTrails/NS_Melee_AnimTrail_NinjaSword
dependencies {
 usedDataInterfaces [&NiagaraDataInterfaceColorCurve, &NiagaraDataInterfaceCurve]
 usedDynamicInputs []
 usedModules [&SystemState]
}
emitters {AnimTrail {
  emitterUpdateScript [
   &EmitterState()
   &FN_SpawnAnimTrail_v2(AnimTrailWidthMode=::FromSecond)
  ]
  particleSpawnScript [
   /Game/Effects/Niagara/Modules/AnimTrail/FN_InitializeAnimTrail/FN_InitializeAnimTrailBoneIndecies(First Bone or Socket Index=1)
   /Game/Effects/Niagara/Modules/AnimTrail/FN_InitializeAnimTrail/FN_InitializeAnimTrail(AnimTrailWidth=1.1, Lifetime=0.35)
   &self:TrailColor(
    Apply TOD colors to Custom Color=0.5
    Color Day:LinearColor=<rgba 0.7912 2.0754 3 1>
    Color Evening:LinearColor=<rgba 2 1.5492 1.3674 1>
    Color Morning:LinearColor=<rgba 1 0.6836 0.5932 1>
    Color Night:LinearColor=<rgba 1.6108 1.8993 2 1>
    Custom Color:LinearColor=$User.Color
    Custom Color Intensity:NiagaraFloat=$User.CustomColorIntensity
    Custom Color Multiplier=2.5
    FortniteNPC_FortniteActiveTimeOfDay:LinearColor=$NPC.FortniteNPC.FortniteActiveTimeOfDay
    Intensity Multiplier=2
    Use Custom Color:NiagaraFloat=$User.UseCustomColor
   ) {module=TrailColor}
  ]
  particleUpdateScript [
   &FN_UpdateAnimTrail_v2()
   &ScaleColor(
    Scale Alpha=&FloatFromCurve(
     FloatCurve=DataInterface() {propertyValues {Curve {keys [
         {arriveTangent=-3.0207, interpMode=RCIM_Cubic, leaveTangent=-3.0207, tangentMode=RCTM_User, time=0, value=1}
         {arriveTangent=-1.1093, interpMode=RCIM_Cubic, leaveTangent=-1.1093, tangentMode=RCTM_User, time=0.3, value=0}
        ]}}}
     CurveIndex=$Particles.Age
     Scale Curve=$User.TrailOpacity
    )
    Scale RGB=&MakeVectorFromLinearColorRGB(Color=&ColorFromCurve(ColorCurve=DataInterface() {propertyValues {
        AlphaCurve {keys []}
        BlueCurve {keys [<at 0 2 RCIM_Cubic>,<at 0.8485 0.1098 RCIM_Cubic>]}
        GreenCurve {keys [<at 0 2 RCIM_Cubic>,<at 0.8485 0.1331 RCIM_Cubic>]}
        RedCurve {keys [<at 0 2 RCIM_Cubic>,<at 0.8485 0.1608 RCIM_Cubic>]}
       }}))
    ScaleA=true
    ScaleRGB=true
   )
   &DynamicMaterialParameters(
    Index 0 Param 1=&FloatFromCurve(FloatCurve=DataInterface() {propertyValues {Curve {keys [<at 0 -0.2>,<at 0.5 10>]}}}, CurveIndex=$Particles.Age)
    Module.Param0WriteEnabled=false
    Module.Param1WriteEnabled=false
    Module.Param2WriteEnabled=false
    Module.Param3WriteEnabled=false
   ) {enabled=false}
  ]
  rendererClasses [&NiagaraRibbonRendererProperties]
  renderers [
   {rendererClass=&NiagaraRibbonRendererProperties}
   {rendererClass=&NiagaraRibbonRendererProperties, rendererIndex=1}
   {rendererClass=&NiagaraRibbonRendererProperties, rendererIndex=2}
  ]
 }}
summary {
 systemName=NS_Melee_AnimTrail_NinjaSword
 emitters [{emitterName=AnimTrail, rendererClasses [&NiagaraRibbonRendererProperties]}]
 userVariables [
  User.Color {type=LinearColor, description="", defaultValue {struct=LinearColor, value=<rgba 0.0694 1 0 1>}}
  User.CustomColorIntensity {type=NiagaraFloat, defaultValue=1, description=""}
  User.RibbonIDOffset {type=NiagaraInt32}
  User.SkeletalMesh {
   type=NiagaraDataInterfaceSkeletalMesh
   defaultValue {
    struct=NiagaraExt_VariableValue_DataInterface
    value {dataInterface=&self:NiagaraDataInterfaceSkeletalMesh_2}
   }
  }
  User.TrailOpacity {type=NiagaraFloat, defaultValue=0.75, description=""}
  User.UseCustomColor {type=NiagaraFloat, defaultValue=0, description=""}
 ]
}
